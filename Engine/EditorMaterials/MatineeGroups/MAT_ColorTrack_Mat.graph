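# Editor material for the Matinee color track graphic, found among the engine's Matinee group materials.
# It uses translucent blending with unlit shading.
# Its graph holds a single sample of the MAT_ColorTrack texture, with no math, parameters or visible link to a material output.
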Material /Engine/EditorMaterials/MatineeGroups/MAT_ColorTrack_Mat
graph {
 properties {blendMode=BLEND_Translucent, pixelDepthOffsetMode=PDOM_Legacy, shadingModel=MSM_Unlit, usedShadingModels=Unlit}
 nodes [TextureSample() {texture=&/MAT_ColorTrack}]
}
