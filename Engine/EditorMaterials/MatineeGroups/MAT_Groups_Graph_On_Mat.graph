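# Editor material in the Matinee groups set, apparently the graph toggle's "on" image, using the Default Lit shading model.
# A single texture sample of MAT_Groups_Graph_On feeds its RGB into the DiffuseColor input of ConvertFromDiffSpec, a function for converting diffuse/specular inputs.
# The graph holds no other logic, and no other input of that function is wired.
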Material /Engine/EditorMaterials/MatineeGroups/MAT_Groups_Graph_On_Mat
graph {
 properties {pixelDepthOffsetMode=PDOM_Legacy, usedShadingModels=Default Lit}
 nodes [MaterialFunctionCall(DiffuseColor=TextureSample() {texture=&/MAT_Groups_Graph_On}.RGB) {materialFunction=&ConvertFromDiffSpec}]
}
