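# Editor material for the Slomo group icon in Matinee, under the engine's MatineeGroups editor materials.
# It samples the MAT_Groups_Slomo texture and feeds its RGB as diffuse color into the ConvertFromDiffSpec material function.
# The material uses masked blending with the Default Lit shading model.
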Material /Engine/EditorMaterials/MatineeGroups/MAT_Groups_Slomo_Mat
graph {
 properties {blendMode=BLEND_Masked, pixelDepthOffsetMode=PDOM_Legacy, usedShadingModels=Default Lit}
 nodes [MaterialFunctionCall(DiffuseColor=TextureSample() {texture=&/MAT_Groups_Slomo}.RGB) {materialFunction=&ConvertFromDiffSpec}]
}
