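# Editor material for the Matinee group trajectory toggle in its "on" state, judging by its name and folder.
# It multiplies the MAT_Groups_Graph_On texture, shared with the graph icon, by a constant green tint.
# The result feeds the DiffuseColor input of a ConvertFromDiffSpec function call under default lit shading.
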Material /Engine/EditorMaterials/MatineeGroups/MAT_Groups_Trajectory_On_Mat
graph {
 properties {pixelDepthOffsetMode=PDOM_Legacy, usedShadingModels=Default Lit}
 nodes [MaterialFunctionCall(DiffuseColor=Multiply(A=TextureSample() {texture=&/MAT_Groups_Graph_On}.RGB, B=Constant3Vector() {constant=<rgba 0 1 0 0>})) {materialFunction=&ConvertFromDiffSpec}]
}
