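# Masked, default-lit editor material that displays the "View Off" texture for Matinee groups.
# The texture's RGB feeds the DiffuseColor input of a ConvertFromDiffSpec function, which converts diffuse/specular inputs to metallic-workflow values.
# The graph shows no other inputs to that function and no opacity mask source.
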Material /Engine/EditorMaterials/MatineeGroups/MAT_Groups_View_Off_Mat
graph {
 properties {blendMode=BLEND_Masked, pixelDepthOffsetMode=PDOM_Legacy, usedShadingModels=Default Lit}
 nodes [MaterialFunctionCall(DiffuseColor=TextureSample() {texture=&/MAT_Groups_View_Off}.RGB) {materialFunction=&ConvertFromDiffSpec}]
}
